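# Post-copy animation Blueprint for the M_MED_PajamaSoar male medium cosmetic body, layering corrective and secondary motion onto its attached parent's pose.
# The copied pose runs through the body's Control Rig, then fixed additive offsets on both foot corrective bones.
# A spring-driven, damped Anim Dynamics body then swings the collar bone.
# Each update maps foot-to-calf yaw into clamped Calf_rot variables, which the shown AnimGraph never reads.
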AnimBlueprint /BRCosmetics/Characters/Player/Male/Medium/Bodies/M_MED_PajamaSoar/Meshes/M_MED_PajamaSoar_AnimBP
target_skeleton=&/M_MED_PajamaSoar_Skeleton
graphs [
 AnimGraph {nodes [
   n1 = Control Rig(Source=Copy Pose From Mesh() {bUseAttachedParent=true}) {
    alphaBoolBlend {blendOption=Linear}
    controlRigAssetReference {blueprintRigClass=&/M_MED_PajamaSoar_CtrlRig.M_MED_PajamaSoar_CtrlRig_C}
   }
   n2 = "Transform (Modify) Bone - Bone: Foot_corrective_L"(ComponentPose=Local To Component(LocalPose=@n1), Alpha=1, Translation="2.035,0.000,2.721", Rotation="0,0,0", Scale="1,1,1") {translationMode=BMM_Additive, alphaBoolBlend {blendOption=Linear}, boneToModify {boneName=Foot_corrective_L}, translation {x=2.035, z=2.721}}
   n3 = "Transform (Modify) Bone - Bone: Foot_corrective_R"(ComponentPose=@n2, Alpha=1, Translation="-0.415,1.295,2.513", Rotation="0,0,0", Scale="1,1,1") {
    translationMode=BMM_Additive
    alphaBoolBlend {blendOption=Linear}
    boneToModify {boneName=Foot_corrective_R}
    translation=<-0.415 1.295 2.513>
   }
   n4 = Anim Dynamics - Bone: dyn_collar!(ComponentPose=@n3, bAlphaBoolEnabled:bool=True, ExternalForce="0,0,0") -> (Pose) {
    alphaInputType=Bool
    angularDampingOverride=0.9999
    angularSpringConstant=500
    bAngularSpring=true
    bGravityOverrideInSimSpace=true
    bOverrideAngularDamping=true
    simulationSpace=BoneRelative
    alphaBoolBlend {blendOption=Linear}
    boundBone {boneName=dyn_collar}
    componentAppliedLinearAccClamp=<3500 3500 3500>
    componentLinearAccScale=<1.2 1.2 1.2>
    componentLinearVelScale=<1.2 1.2 1.2>
    physicsBodyDefinitions [{
      sphereCollisionRadius=0
      boundBone {boneName=dyn_collar}
      boxExtents=<15 20 20>
      constraintSetup {angularLimitsMax=<0 15 0>, angularLimitsMin=<0 -15 0>, angularTarget=<0 0 0>}
      localJointOffset=<4 0 0>
     }]
    relativeSpaceBone {boneName=root}
   }
   Output Pose(Result=Component To Local(ComponentPose=@n4)) {layerGroup=None}
  ]}
 EventGraph {nodes [
   Event BlueprintUpdateAnimation()
   n1 = GetOwnerSkeletalMeshComponent()
   > Set Calf_rot_L(Calf_rot_L:double=MapRangeClamped(
     Value=GetDeltaTransformFromRefPose!(self=@n1, BoneName=foot_l, BaseName=calf_l) -> (
      ReturnValue_Location:Vector
      ReturnValue_Rotation_Roll:float
      ReturnValue_Rotation_Pitch:float
      ReturnValue_Rotation_Yaw:float
      ReturnValue_Scale:Vector
     ).ReturnValue_Rotation_Yaw
     InRangeA=0
     InRangeB=60
    )) -> (Output_Get:double)
   n2 = GetOwnerSkeletalMeshComponent()
   > Set Calf_rot_R(Calf_rot_R:double=MapRangeClamped(Value=GetDeltaTransformFromRefPose(self=@n2, BoneName=foot_r, BaseName=calf_r).ReturnValue_Rotation_Yaw, InRangeA=0, InRangeB=60)) -> (Output_Get:double)
  ]}
]
variables [
 Calf_rot_L {type=number}
 Calf_rot_R {type=number}
]
