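# Physics asset for the face accessory part of the male medium PajamaSoar player cosmetic, defining bodies and joints for its hair bones.
# Sphere bodies cover the hair bones, alongside a small head capsule and a large spine_05 capsule for the upper body.
# Most hair bodies attach directly to the head with limited swing and very tight twist.
# Only hair4 chains through base_hair4, with a narrower first swing; base_hair4-to-head and head-to-spine_05 joints rotate freely.
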PhysicsAsset /BRCosmetics/Characters/Player/Male/Medium/Bodies/M_MED_PajamaSoar/Meshes/Parts/M_MED_PajamaSoar_FaceAcc_Physics
bodies {
 base_hair4 [{shapeType=Sphere, radius=0.505, shapeName=base_hair4_sphere}]
 hair1 [{shapeType=Sphere, radius=0.505, shapeName=hair1_sphere, center=<2 0 0>}]
 hair10 [{shapeType=Sphere, radius=0.505, shapeName=hair10_sphere, center=<2 0 0>}]
 hair11 [{shapeType=Sphere, radius=2.5581, shapeName=hair11_sphere, center=<2 0 0>}]
 hair12 [{shapeType=Sphere, radius=1.8326, shapeName=hair12_sphere, center=<2 0 0>}]
 hair13 [{shapeType=Sphere, radius=1.6659, shapeName=hair13_sphere, center=<2 0 0>}]
 hair14 [{shapeType=Sphere, radius=2.2542, shapeName=hair14_sphere, center=<2 0 0>}]
 hair15 [{shapeType=Sphere, radius=3.5133, shapeName=hair15_sphere, center=<2 0 0>}]
 hair16 [{shapeType=Sphere, radius=1.9075, shapeName=hair16_sphere, center=<2 0 0>}]
 hair17 [{shapeType=Sphere, radius=2.4591, shapeName=hair17_sphere, center=<2 0 0>}]
 hair18 [{shapeType=Sphere, radius=2.5997, shapeName=hair18_sphere, center=<2 0 0>}]
 hair19 [{shapeType=Sphere, radius=1.9192, shapeName=hair19_sphere, center=<2 0 0>}]
 hair2 [{shapeType=Sphere, radius=1.864, shapeName=hair2_sphere, center=<2 0 0>}]
 hair20 [{shapeType=Sphere, radius=1.1854, shapeName=hair20_sphere, center=<2 0 0>}]
 hair21 [{shapeType=Sphere, radius=1.5996, shapeName=hair21_sphere, center=<2 0 0>}]
 hair22 [{shapeType=Sphere, radius=0.505, shapeName=hair22_sphere, center=<8.0362 -0.7843 0.496>}]
 hair3 [{shapeType=Sphere, radius=1.6883, shapeName=hair3_sphere, center=<2 0 0>}]
 hair4 [{shapeType=Sphere, radius=2.9852, shapeName=hair4_sphere, center=<10.7281 0.1719 0>}]
 hair5 [{shapeType=Sphere, radius=3.0346, shapeName=hair5_sphere, center=<5.5873 0 0>}]
 hair7 [{shapeType=Sphere, radius=0.505, shapeName=hair7_sphere, center=<2 0 0>}]
 hair8 [{shapeType=Sphere, radius=2.6039, shapeName=hair8_sphere, center=<2 0 0>}]
 hair9 [{shapeType=Sphere, radius=0.505, shapeName=hair9_sphere, center=<2 0 0>}]
 head [{shapeType=Capsule, length=0, radius=1.3406, center=<8.633 0.4979 -1.0588>, rotation=<rot 0 -56.1497 -69.965>}]
 spine_05 [{shapeType=Capsule, length=19, radius=11, shapeName=spine_05_capsule, center=<7.1421 4.0273 0>, rotation=<rot 0 -90 0>}]
}
constraints [
 {bone1Name=hair1, bone2Name=head, swing1LimitDegrees=20, swing2LimitDegrees=20, twistLimitDegrees=1}
 {bone1Name=hair2, bone2Name=head, swing1LimitDegrees=20, swing2LimitDegrees=20, twistLimitDegrees=1}
 {bone1Name=hair3, bone2Name=head, swing1LimitDegrees=20, swing2LimitDegrees=20, twistLimitDegrees=1}
 {bone1Name=hair4, bone2Name=base_hair4, swing1LimitDegrees=3, swing2LimitDegrees=20, twistLimitDegrees=1}
 {bone1Name=hair5, bone2Name=head, swing1LimitDegrees=20, swing2LimitDegrees=20, twistLimitDegrees=1}
 {bone1Name=hair7, bone2Name=head, swing1LimitDegrees=20, swing2LimitDegrees=20, twistLimitDegrees=1}
 {bone1Name=hair8, bone2Name=head, swing1LimitDegrees=20, swing2LimitDegrees=20, twistLimitDegrees=1}
 {bone1Name=hair9, bone2Name=head, swing1LimitDegrees=20, swing2LimitDegrees=20, twistLimitDegrees=1}
 {bone1Name=hair10, bone2Name=head, swing1LimitDegrees=20, swing2LimitDegrees=20, twistLimitDegrees=1}
 {bone1Name=hair11, bone2Name=head, swing1LimitDegrees=20, swing2LimitDegrees=20, twistLimitDegrees=1}
 {bone1Name=hair12, bone2Name=head, swing1LimitDegrees=20, swing2LimitDegrees=20, twistLimitDegrees=1}
 {bone1Name=hair13, bone2Name=head, swing1LimitDegrees=20, swing2LimitDegrees=20, twistLimitDegrees=1}
 {bone1Name=hair14, bone2Name=head, swing1LimitDegrees=20, swing2LimitDegrees=20, twistLimitDegrees=1}
 {bone1Name=hair15, bone2Name=head, swing1LimitDegrees=20, swing2LimitDegrees=20, twistLimitDegrees=1}
 {bone1Name=hair16, bone2Name=head, swing1LimitDegrees=20, swing2LimitDegrees=20, twistLimitDegrees=1}
 {bone1Name=hair17, bone2Name=head, swing1LimitDegrees=20, swing2LimitDegrees=20, twistLimitDegrees=1}
 {bone1Name=hair18, bone2Name=head, swing1LimitDegrees=20, swing2LimitDegrees=20, twistLimitDegrees=1}
 {bone1Name=hair19, bone2Name=head, swing1LimitDegrees=20, swing2LimitDegrees=20, twistLimitDegrees=1}
 {bone1Name=hair20, bone2Name=head, swing1LimitDegrees=20, swing2LimitDegrees=20, twistLimitDegrees=1}
 {bone1Name=hair21, bone2Name=head, swing1LimitDegrees=20, swing2LimitDegrees=20, twistLimitDegrees=1}
 {bone1Name=hair22, bone2Name=head, swing1LimitDegrees=20, swing2LimitDegrees=20, twistLimitDegrees=1}
 {
  bone1Name=head
  bone2Name=spine_05
  swing1LimitDegrees=45
  swing1Motion=Free
  swing2LimitDegrees=45
  swing2Motion=Free
  twistLimitDegrees=45
  twistMotion=Free
 }
 {
  bone1Name=base_hair4
  bone2Name=head
  swing1LimitDegrees=45
  swing1Motion=Free
  swing2LimitDegrees=45
  swing2Motion=Free
  twistLimitDegrees=45
  twistMotion=Free
 }
]
