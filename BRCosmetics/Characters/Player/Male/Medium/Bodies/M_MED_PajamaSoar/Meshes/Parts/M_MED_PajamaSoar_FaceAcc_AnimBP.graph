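# Animation Blueprint for the PajamaSoar character's face accessory, copying the parent mesh pose and adding secondary hair motion.
# An additive offset on base_hair4 scales with head yaw relative to spine_05, then feeds a half-strength RigidBody simulation in spine_05 space with clamped motion.
# The cached result blends with DefaultSlot on the hair2 and hair3 branches, weighted by the Frontend_Idle_Bypass curve read in the front end.
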AnimBlueprint /BRCosmetics/Characters/Player/Male/Medium/Bodies/M_MED_PajamaSoar/Meshes/Parts/M_MED_PajamaSoar_FaceAcc_AnimBP
target_skeleton=&/M_MED_PajamaSoar_FaceAcc_Skeleton
graphs [
 AnimGraph {nodes [
   Output Pose(Result=Layered blend per bone(
     BasePose=Use cached pose 'startPose'()
     BlendPoses_0=Slot 'DefaultSlot'()
     BlendWeights_0=Get Custom Idle Is Playing() -> (Custom Idle Is Playing:double)
    ) {layerSetup [{branchFilters [{boneName=hair2}, {boneName=hair3}]}]}) {layerGroup=None}
   n1 = "Transform (Modify) Bone - Bone: base_hair4"(
    ComponentPose=Local To Component(LocalPose=Copy Pose From Mesh() {bUseAttachedParent=true})
    Alpha=Get Head Rot() -> (Head Rot:double)
    Translation="5,5,0"
    Rotation="0,30,0"
    Scale="1,1,1"
   ) {
    rotationMode=BMM_Additive
    rotationSpace=BCS_BoneSpace
    translationMode=BMM_Additive
    translationSpace=BCS_BoneSpace
    alphaBoolBlend {blendOption=Linear}
    boneToModify {boneName=base_hair4}
    rotation {yaw=30}
    translation=<5 5>
   }
   n2 = RigidBody!(ComponentPose=@n1, Alpha=0.5) -> (Pose) {
    bOverrideWorldGravity=true
    overridePhysicsAsset=&/M_MED_PajamaSoar_FaceAcc_Physics
    simulationSpace=BaseBoneSpace
    alphaBoolBlend {blendOption=Linear}
    baseBoneRef {boneName=spine_05}
    componentAppliedLinearAccClamp=<2500 2500 2500>
    overrideWorldGravity {z=-980}
    simSpaceSettings {maxAngularAcceleration=65, maxAngularVelocity=15, maxLinearAcceleration=1000, maxLinearVelocity=3, velocityScaleZ=0.1}
   }
   Save cached pose 'startPose'(Pose=Component To Local(ComponentPose=@n2)) {cacheName=startPose}
  ]}
 EventGraph {nodes [
   Event BlueprintUpdateAnimation()
   n1 = GetOwnerSkeletalMeshComponent()
   > Set Head Rot(Head Rot:double=MapRangeClamped(
     Value=GetDeltaTransformFromRefPose!(self=@n1, BoneName=head, BaseName=spine_05) -> (
      ReturnValue_Location:Vector
      ReturnValue_Rotation_Roll:float
      ReturnValue_Rotation_Pitch:float
      ReturnValue_Rotation_Yaw:float
      ReturnValue_Scale:Vector
     ).ReturnValue_Rotation_Yaw
     InRangeA=0
     InRangeB=-90
    )) -> (Output_Get:double)
   > Branch(Condition=Get bIsInFrontEnd())
   > Set Custom Idle Is Playing(Custom Idle Is Playing:double=GetCurveValue(self=GetOwnerFrontEndAnimInstance(), CurveName=Frontend_Idle_Bypass)) -> (Output_Get:double)
  ]}
]
variables [
 Custom Idle Is Playing {type=number}
 Head Rot {type=number}
]
